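; Camera modifier for the Thermal Flopper consumable, applying a post-process material to the player's view.
; It overrides BlueprintModifyPostProcess to return full blend weight with post-process settings whose weighted blendables are built from the PPMaterial variable.
; The delta time input is unused, so the modifier itself does no fading or animation.
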(fn BlueprintModifyPostProcess
  (params (DeltaTime :type Float%20%28single-precision%29))
  (return 1.0 (Utilities|Struct|MakePostProcessSettings (Utilities|Struct|MakeWeightedBlendables (Variables|Default|GetPPMaterial)))))
